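# Looping ambient SoundCue for the captured state of the BeetCalf Tower POI console.
# A Mixer layers the BeetCalf captured loop as the louder layer over a quieter, half-pitched Radio Tower console active loop.
# An Enveloper on the root fades the mix in along a rising volume curve over roughly two seconds.
# Exterior ambient attenuation spatializes it, and the cue references the Radio Tower console loop concurrency set.
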SoundCue /BeetCalf/Audio/POI/BeetCalfTower/BeetCalf_Console_Captured_Loop_Cue
root=Enveloper_0
nodes [
 {
  id=Enveloper_0
  children [Mixer_0]
  properties {
   pitchCurve {editorCurveData {keys [{time=0}]}}
   volumeCurve {editorCurveData {keys [<at 0 0.01>,<at 0.3125 0.0741>,<at 1.0258 0.2024>,<at 1.5541 0.4562>,<at 1.9743 1>]}}
  }
 }
 {id=Mixer_0, children [WavePlayer_2,Modulator_1], properties {inputVolume [1.3,0.3]}}
 {id=Modulator_1, children [WavePlayer_0], properties {pitchMax=0.5, pitchMin=0.5, volumeMax=1, volumeMin=1}}
 {id=WavePlayer_0, properties {bLooping=true, soundWaveAssetPtr=/MidMatchRadioTowerGameplay/Sound/RadioTowerConsole_Active_Loop}}
 {id=WavePlayer_2, properties {bLooping=true, soundWaveAssetPtr=&/BeetCalf_Console_Captured_Loop}}
]
properties {
 attenuationSettings=/Game/Sounds/Apollo/Biomes/Interior_Banks/AmbientAudio_Exterior_Att
 bEnableBusSends=false
 maxDistance=3000
 soundClassObject=&Fort_SFX_InGame
 volumeMultiplier=1
 concurrencyOverrides {maxCount=1, resolutionRule=PreventNew}
 concurrencySet [/MidMatchRadioTowerGameplay/Sound/RadioTowerConsole_Loop_Conc]
}
