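# Sound cue for the interact sound of the BeetCalf console at the BeetCalfTower POI, in the in-game SFX sound class.
# A single wave player feeds the BeetCalf_Console_Interact wave straight to the output, with a boosted volume and bus sends disabled.
# Spatialization reuses the radio tower dishes attenuation asset from the MidMatchRadioTowerGameplay content.
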SoundCue /BeetCalf/Audio/POI/BeetCalfTower/BeetCalf_Console_Interact_Cue
root=WavePlayer_0
nodes [{id=WavePlayer_0, properties {soundWaveAssetPtr=&/BeetCalf_Console_Interact}}]
properties {
 attenuationSettings=/MidMatchRadioTowerGameplay/Sound/MMObj_RadioTowerDishes_Att
 bEnableBusSends=false
 maxDistance=25000
 soundClassObject=&Fort_SFX_InGame
 volumeMultiplier=1.5
}
