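# Sound cue for the looping active sound of the BeetCalf radio tower point of interest, played as an in-game SFX.
# A looping wave player feeds a fixed half-second delay, then an enveloper that ramps volume from silence to slightly above unity over about 2.5 seconds.
# It uses boosted volume, distant radio-tower attenuation and the shared RadioTowerPower loop concurrency set, with bus sends disabled.
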SoundCue /BeetCalf/Audio/POI/BeetCalfTower/BeetCalf_Tower_Activate_Loop_Cue
root=Enveloper_2
nodes [
 {id=Delay_0, children [WavePlayer_3], properties {delayMax=0.5, delayMin=0.5}}
 {
  id=Enveloper_2
  children [Delay_0]
  properties {pitchCurve {editorCurveData {keys [{time=0}]}}, volumeCurve {editorCurveData {keys [<at 0 0>,<at 2.4953 1.1196>]}}}
 }
 {id=WavePlayer_3, properties {bLooping=true, soundWaveAssetPtr=&/BeetCalf_Tower_Active_Loop}}
]
properties {
 attenuationSettings=&/MMObj_BeetCalf_RadioTower_Distant_Att
 bEnableBusSends=false
 maxDistance=13000
 soundClassObject=&Fort_SFX_InGame
 volumeMultiplier=1.6
 concurrencyOverrides {maxCount=1, resolutionRule=PreventNew}
 concurrencySet [/MidMatchRadioTowerGameplay/Sound/RadioTowerPower_Loop_Conc]
}
