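# Sound cue for the ping of the BeetCalf Tower point of interest, built on a UI ping sound.
# A single wave player feeds a modulator that applies no pitch or volume variation, with the cue volume boosted.
# It borrows attenuation from the mid-match radio tower gameplay content and uses the critical in-game UI sound class.
# A concurrency override allows one instance at a time, stopping the oldest.
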SoundCue /BeetCalf/Audio/POI/BeetCalfTower/BeetCalf_Tower_Ping_Cue
root=Modulator_5
nodes [
 {id=Modulator_5, children [WavePlayer_2], properties {pitchMax=1, pitchMin=1, volumeMax=1, volumeMin=1}}
 {id=WavePlayer_2, properties {soundWaveAssetPtr=&/BeetCalf_Tower_UI_Ping}}
]
properties {
 attenuationSettings=/MidMatchRadioTowerGameplay/Sound/MMObj_UI_RadioTower_Ping_Att
 bOverrideConcurrency=true
 maxDistance=2097152
 soundClassObject=&Fort_UI_InGame_Critical
 volumeMultiplier=2
 concurrencyOverrides {maxCount=1, resolutionRule=StopOldest}
}
