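# Looping exterior ambience cue for wind around a metal tower antenna at the BeetCalf tower POI, in the outdoor ambient sound class.
# It mixes a howling cold wind loop with a louder-weighted antenna wire wind loop, which has random pitch modulation and a short fade-in.
# Nested looping envelopers of differing lengths vary pitch and volume over time.
# Capsule attenuation adds low-pass and listener focus, alongside limited concurrency and an exterior bus send.
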SoundCue /BeetCalf/Audio/POI/BeetCalfTower/Spot_BeefCalf_Metal_Tower_Antenna_Wind_Ext_Loop_Cue
root=QualityLevel_0
nodes [
 {
  id=Enveloper_0
  children [WavePlayer_0]
  properties {
   bLoop=true
   bLoopIndefinitely=true
   loopEnd=30
   pitchCurve {editorCurveData {keys [{interpMode=RCIM_Cubic, time=0}, <at 15.9468 0.9181 RCIM_Cubic>, {interpMode=RCIM_Cubic, time=30}]}}
   volumeCurve {editorCurveData {keys [<at 0 1>]}}
  }
 }
 {
  id=Enveloper_1
  children [Modulator_1]
  properties {
   bLoop=true
   bLoopIndefinitely=true
   loopEnd=30
   pitchCurve {editorCurveData {keys [{interpMode=RCIM_Cubic, time=0}, <at 15.9468 0.9 RCIM_Cubic>, {interpMode=RCIM_Cubic, time=30}]}}
   volumeCurve {editorCurveData {keys [<at 0 1>]}}
  }
 }
 {
  id=Enveloper_2
  children [Enveloper_1]
  properties {
   bLoop=true
   bLoopIndefinitely=true
   loopEnd=120
   pitchCurve {editorCurveData {keys [{time=0}]}}
   volumeCurve {editorCurveData {keys [
      {arriveTangent=0.0001, interpMode=RCIM_Cubic, leaveTangent=0.0082, tangentMode=RCTM_User, time=0.4639, value=0.3721}
      {arriveTangent=0.0321, interpMode=RCIM_Cubic, leaveTangent=0.0321, time=10, value=0.6}
      {arriveTangent=-0.003, interpMode=RCIM_Cubic, leaveTangent=-0.003, time=20, value=1}
      <at 45.7834 0.4918 RCIM_Cubic>
      {arriveTangent=0.0153, interpMode=RCIM_Cubic, leaveTangent=0.0153, time=60, value=1}
      {arriveTangent=-0.0202, interpMode=RCIM_Cubic, leaveTangent=-0.0202, tangentMode=RCTM_User, time=82.1697, value=1.05}
      <at 119.9639 0.4918>
     ]}}
  }
 }
 {
  id=Enveloper_29
  children [Enveloper_2]
  properties {
   pitchCurve {editorCurveData {keys [{time=0}]}}
   volumeCurve {editorCurveData {keys [{arriveTangent=3.4293, interpMode=RCIM_Cubic, leaveTangent=3.4293, tangentMode=RCTM_User, time=0, value=0}, <at 0.4 1 RCIM_Cubic>]}}
  }
 }
 {
  id=Enveloper_3
  children [Enveloper_0]
  properties {
   bLoop=true
   bLoopIndefinitely=true
   loopEnd=20
   pitchCurve {editorCurveData {keys [{time=0}]}}
   volumeCurve {editorCurveData {keys [
      {arriveTangent=-0.01, interpMode=RCIM_Cubic, time=0, value=1}
      <at 10 1.2 RCIM_Cubic>
      {interpMode=RCIM_Cubic, leaveTangent=0.01, time=20, value=1}
     ]}}
  }
 }
 {id=Mixer_0, children [Enveloper_3,Enveloper_29], properties {inputVolume [0.6,1]}}
 {id=Modulator_1, children [WavePlayer_2], properties {pitchMax=1.2, pitchMin=1}}
 {id=QualityLevel_0, children [Mixer_0,Mixer_0,null,null]}
 {id=WavePlayer_0, properties {bLooping=true, soundWaveAssetPtr=&/Spot_Wind_Howling_Cold_Ext_Loop}}
 {id=WavePlayer_2, properties {bLooping=true, soundWaveAssetPtr=&/Spot_Antenna_Ambient_Wire_Wind_Loop}}
]
properties {
 bOverrideAttenuation=true
 bOverrideConcurrency=true
 maxDistance=3140.5359
 soundClassObject=&Fort_Ambient_Outdoors_Athena
 volumeMultiplier=0.401
 attenuationOverrides {
  attenuationShape=Capsule
  bAttenuateWithLPF=true
  bEnableFocusInterpolation=true
  bEnableListenerFocus=true
  falloffDistance=1794.1893
  focusReleaseInterpSpeed=0.9
  lPFFrequencyAtMax=750
  lPFRadiusMax=4000
  lPFRadiusMin=1000
  nonFocusVolumeAttenuation=0.75
  attenuationShapeExtents=<1346.3466 337.2983>
 }
 busSends [{audioBus=null, minSendLevel=0.6, sendLevel=1, soundSourceBus=&AmbientAudio_Exterior_SourceBus}]
 concurrencyOverrides {maxCount=3}
 concurrencySet [null]
}
