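# Sound Cue for an exterior looping radio comms chatter ambience at the BeetCalf tower POI.
# A quiet looping glitchy comms bed mixes with a looped chain that plays one of eight random chatter clips after short random delays.
# An envelope fades the mix in, feeding overridden capsule attenuation with listener focus.
# The outdoor-ambient cue overrides concurrency to limit simultaneous instances and sends to the default reverb submix.
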SoundCue /BeetCalf/Audio/POI/BeetCalfTower/Spot_BeefCalf_Radio_Comms_Chatter_Loop_Ext_Cue
root=Attenuation_0
nodes [
 {
  id=Attenuation_0
  children [Enveloper_0]
  properties {
   attenuationSettings=/Game/Sounds/Helios/Environment_Actors/Console_RadioCommunication/Spot_Console_RadioCommunication_Loop_Att
   bOverrideAttenuation=true
   attenuationOverrides {
    attenuationShape=Capsule
    bEnableListenerFocus=true
    falloffDistance=5000
    nonFocusVolumeAttenuation=0.4
    attenuationShapeExtents=<2000 1500>
   }
  }
 }
 {id=Delay_0, children [Random_1], properties {delayMax=2, delayMin=1}}
 {
  id=Enveloper_0
  children [Mixer_0]
  properties {pitchCurve {editorCurveData {keys [{time=0}]}}, volumeCurve {editorCurveData {keys [<at 0 0>,<at 1.5 1>]}}}
 }
 {id=Looping_0, children [Delay_0]}
 {id=Mixer_0, children [WavePlayer_1,Looping_0], properties {inputVolume [0.15,1]}}
 {
  id=Random_1
  children [WavePlayer_9, WavePlayer_10, WavePlayer_11, WavePlayer_12, WavePlayer_13, WavePlayer_14, WavePlayer_15, WavePlayer_16]
  properties {weights [1,1,1,1,1,1,1,1]}
 }
 {id=WavePlayer_1, properties {bLooping=true, soundWaveAssetPtr=&/Spot_BeefCallf_Comms_Glitchy_Loop}}
 {id=WavePlayer_10, properties {soundWaveAssetPtr=&/Spot_BeefCallf_Chatter_02}}
 {id=WavePlayer_11, properties {soundWaveAssetPtr=&/Spot_BeefCallf_Chatter_03}}
 {id=WavePlayer_12, properties {soundWaveAssetPtr=&/Spot_BeefCallf_Chatter_04}}
 {id=WavePlayer_13, properties {soundWaveAssetPtr=&/Spot_BeefCallf_Chatter_05}}
 {id=WavePlayer_14, properties {soundWaveAssetPtr=&/Spot_BeefCallf_Chatter_06}}
 {id=WavePlayer_15, properties {soundWaveAssetPtr=&/Spot_BeefCallf_Chatter_07}}
 {id=WavePlayer_16, properties {soundWaveAssetPtr=&/Spot_BeefCallf_Chatter_08}}
 {id=WavePlayer_9, properties {soundWaveAssetPtr=&/Spot_BeefCallf_Chatter_01}}
]
properties {
 bOverrideConcurrency=true
 maxDistance=7000
 soundClassObject=&Fort_Ambient_Outdoors_Athena
 volumeMultiplier=0.501
 concurrencyOverrides {maxCount=2, volumeScale=0.5}
 soundSubmixSends [{minSendLevel=0, soundSubmix=&Fortnite_Reverb_Default}]
}
